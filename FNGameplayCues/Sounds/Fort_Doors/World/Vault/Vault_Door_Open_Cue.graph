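# Sound cue that plays the opening sound for a vault door, filed under world door gameplay cue sounds.
# A single wave player feeds the Vault_Door_Open_01 wave straight to the output, with no randomization or modulation.
# The cue uses the Fort_Doors sound class and shared door attenuation, with a slight volume boost and virtualization disabled.
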SoundCue /FNGameplayCues/Sounds/Fort_Doors/World/Vault/Vault_Door_Open_Cue
root=WavePlayer_1
nodes [{id=WavePlayer_1, properties {soundWaveAssetPtr=&/Vault_Door_Open_01}}]
properties {
 attenuationSettings=/Game/Sounds/Attenuation/Fort_Doors_Att
 maxDistance=2400
 soundClassObject=&Fort_Doors
 virtualizationMode=Disabled
 volumeMultiplier=1.25
}
