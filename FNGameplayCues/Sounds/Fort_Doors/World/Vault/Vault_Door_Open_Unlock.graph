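# Sound cue for a world vault door opening, combining an unlock sound with an alarm.
# A Mixer plays both wave players together at equal volume, with no randomization or delay.
# It uses the Fort_Doors sound class and shared door attenuation, with a volume boost and play-when-silent virtualization.
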SoundCue /FNGameplayCues/Sounds/Fort_Doors/World/Vault/Vault_Door_Open_Unlock
root=Mixer_0
nodes [
 {id=Mixer_0, children [WavePlayer_0,WavePlayer_1], properties {inputVolume [1,1]}}
 {id=WavePlayer_0, properties {soundWaveAssetPtr=&/Vault_Door_Open_Unlock_01}}
 {id=WavePlayer_1, properties {soundWaveAssetPtr=&/Vault_Door_Open_Alarm_01}}
]
properties {
 attenuationSettings=/Game/Sounds/Attenuation/Fort_Doors_Att
 maxDistance=2400
 soundClassObject=&Fort_Doors
 virtualizationMode=PlayWhenSilent
 volumeMultiplier=1.25
}
